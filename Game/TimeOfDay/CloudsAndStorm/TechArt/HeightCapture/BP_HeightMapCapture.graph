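# Editor utility Blueprint that sets up an orthographic scene capture for a landscape height map, running CaptureHeight from its construction script.
# A downward box trace, ignoring blocking and kill-Z volumes, stores the hit height in MaxHeight.
# If exactly one camera is tagged MinimapCaptureCamera, the capture component copies its transform and ortho width.
# It then writes MaxHeight in MPC_Helios_LandscapeInfo as a hard-coded placeholder-like constant, not the traced value.
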EditorUtilityBlueprint /Game/TimeOfDay/CloudsAndStorm/TechArt/HeightCapture/BP_HeightMapCapture
graphs [
 CaptureHeight {nodes [
   n1 = Cast To LandscapeComponent!(execute, Object=Get RootComponent!(self:Actor Object Reference) -> (RootComponent:Scene Component Object Reference)) -> (then, CastFailed, AsLandscape Component:Landscape Component Object Reference) {is=title K2Node_DynamicCast}
   > PrintString(InString="To String (Vector)"!(InVec:Vector=Get Actor Bounds(bOnlyCollidingComponents=false)) -> (ReturnValue:String), Duration=10)
   CaptureHeight() {is=title K2Node_FunctionEntry}
   > n2 = GetEditorWorld(self=Get UnrealEditorSubsystem())
   > n3 = GetAllActorsOfClass!(execute, WorldContextObject:Object Reference=@n2, ActorClass:Actor Class Reference=/Script/Engine.BlockingVolume) -> (then, OutActors:Array of Blocking Volume Object References)
   n4 = Get ActorArray() -> (ActorArray:Array of Actor Object References)
   > Append Array!(execute, TargetArray:Array of Actor Object References=@n4, SourceArray:Array of Actor Object References=@n3) -> (then)
   > n5 = GetAllActorsOfClass(ActorClass=/Script/Engine.KillZVolume) -> (OutActors:Array of Kill ZVolume Object References)
   > Append Array(TargetArray=@n4, SourceArray=@n5)
   > n6 = Box Trace For Objects!(
    execute
    WorldContextObject:Object Reference=@n2
    Start:Vector="8500,508,36000"
    End:Vector="8500,508,0"
    HalfSize:Vector="200000,200000,5"
    Orientation:Rotator="0, 0, 0"
    ObjectTypes:Array of EObjectTypeQuery Enums=Make Array("[0]":EObjectTypeQuery Enum=ObjectTypeQuery1) -> (Array:Array of EObjectTypeQuery Enums)
    bTraceComplex:bool=true
    ActorsToIgnore:Array of Actor Object References=Get ActorArray()
    DrawDebugType:EDrawDebugTrace Enum=ForDuration
    bIgnoreSelf:bool=true
    TraceColor:Linear Color Structure=(R=1,G=0,B=0,A=1)
    TraceHitColor:Linear Color Structure=(R=0,G=1,B=0,A=1)
    DrawTime:float=12
   ) -> (then, OutHit:Hit Result Structure, ReturnValue:bool)
   n7 = BreakHitResult!(Hit:"Hit Result Structure (by ref)"=@n6) -> (
    bBlockingHit:bool
    bInitialOverlap:bool
    Time:float
    Distance:float
    Location:Vector
    ImpactPoint:Vector
    Normal:Vector
    ImpactNormal:Vector
    PhysMat:Physical Material Object Reference
    HitActor:Actor Object Reference
    HitComponent:Primitive Component Object Reference
    HitBoneName:Name
    BoneName:Name
    HitItem:int
    ElementIndex:int
    FaceIndex:int
    TraceStart:Vector
    TraceEnd:Vector
   )
   > Branch(Condition=@n7)
   > Set MaxHeight(MaxHeight:double=BreakVector(InVec=@n7.Location).Z) -> (Output_Get:double)
   > n8 = GetEditorWorld(self=Get UnrealEditorSubsystem())
   > n9 = GetAllActorsOfClassWithTag!(execute, WorldContextObject:Object Reference=@n8, ActorClass:Actor Class Reference=/Script/Engine.CameraActor, Tag:Name=MinimapCaptureCamera) -> (then, OutActors:Array of Camera Actor Object References)
   > Branch(Condition="Equal (Integer)"(A=Length(TargetArray:Array of Camera Actor Object References=@n9), B=1))
   n10 = "Get (a copy)"(Array:Array of Camera Actor Object References=@n9) -> (Output:Camera Actor Object Reference)
   n11 = GetComponentByClass!(self:Actor Object Reference=@n10, ComponentClass:Actor Component Class Reference=/Script/Engine.CameraComponent) -> (ReturnValue:Camera Component Object Reference)
   n12 = Get HeightCaptureComponent() -> (HeightCaptureComponent:Scene Capture Component 2D Object Reference)
   > SetAbsolute!(
    execute
    self:Scene Component Object Reference=[@n11, @n12]
    bNewAbsoluteLocation:bool=true
    bNewAbsoluteRotation:bool=true
    bNewAbsoluteScale:bool=true
   ) -> (then)
   n13 = BreakTransform(InTransform=Get Actor Transform(self=@n10))
   n14 = BreakVector(InVec=@n13)
   > Set World Transform!(
    execute
    self:Scene Component Object Reference=[@n11, @n12, @n12]
    NewTransform:"Transform (by ref)"=MakeTransform(Location=MakeVector(X=@n14, Y=@n14.Y, Z=@n14.Z), Rotation=@n13.Rotation, Scale=@n13.Scale)
    bSweep:bool=false
    bTeleport:bool=false
   ) -> (then, SweepHitResult:Hit Result Structure) {is=title K2Node_CallFunction}
   > n15 = Set Ortho Width(Ortho Width:float=Get OrthoWidth!(self:Camera Component Object Reference=@n11) -> (OrthoWidth:float)) -> (Output_Get:float)
   > Set OrthoWidth!(execute, OrthoWidth:float=@n15, self:Scene Capture Component 2D Object Reference=@n12) -> (then, Output_Get:float)
   > GetEditorWorld(self=Get UnrealEditorSubsystem())
   > SetScalarParameterValue!(
    execute
    WorldContextObject
    Collection:Material Parameter Collection Object Reference=/Game/Athena/Helios/Environments/Landscape/MPC/MPC_Helios_LandscapeInfo
    ParameterName:Name=MaxHeight
    ParameterValue:float=3.14
   ) -> (then) {is=title K2Node_CallMaterialParameterCollectionFunction}
   GetAllActorsOfClass(WorldContextObject=@n2, ActorClass=&CameraActor) -> (OutActors:Array of Camera Actor Object References)
   Get World Transform!(self:Scene Component Object Reference=Get CameraComponent!(self:Camera Actor Object Reference=@n10) -> (CameraComponent:Camera Component Object Reference)) -> (ReturnValue:Transform) {is=title K2Node_CallFunction}
   MakeLiteralName(Value=OrthoWidth)
   MakeTransform(Location=MakeVector(X=@n14, Y=@n14.Y), Rotation="0, 0, 0", Scale="3200,3200,1")
   MakeTransform(Location=MakeVector(X=@n14, Y=@n14.Y, Z=36000), Rotation="0, 0, 0", Scale="3200,3200,1")
   PrintString(InString="To String (Vector)"(InVec=@n7.Location), Duration=10)
   n16 = GetComponentBounds!(Component:Scene Component Object Reference=@n1) -> (Origin:Vector, BoxExtent:Vector, SphereRadius:float)
   "To String (Vector)"(InVec=@n16)
   SetVectorParameterValue!(
    execute
    WorldContextObject=Get MapMPC() -> (MapMPC:Material Parameter Collection Object Reference)
    Collection:Material Parameter Collection Object Reference
    ParameterName:Name=None
    ParameterValue:"Linear Color Structure (by ref)"
   ) -> (then) {is=title K2Node_CallMaterialParameterCollectionFunction}
   Get AssetEditorSubsystem()
   Get EditorAssetSubsystem() -> (ReturnValue:Editor Asset Subsystem Object Reference) {is=title K2Node_GetEditorSubsystem}
   "Equal (GameplayTagContainer)"(A:Wildcard, B:Wildcard) -> (ReturnValue:bool=~) {is=title K2Node_PromotableOperator}
   vector + vector(A=@n16, B=@n16.BoxExtent)
  ]}
 EventGraph {nodes [
   Event BeginPlay()
   Event ActorBeginOverlap()
   Event Tick!() -> (OutputDelegate, then, DeltaSeconds:float)
   Get UnrealEditorSubsystem()
  ]}
 UserConstructionScript {nodes [
   Construction Script()
   > CaptureHeight() {is=title K2Node_CallFunction}
  ]}
]
variables [
 ActorArray {type="array<Actor>"}
 MapMPC {type=MaterialParameterCollection, default=&MPC_Helios_LandscapeInfo}
 MaxHeight {type=number}
 Ortho Width {type=number}
]
